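; Deprecated Blueprint function set for controlling managed Niagara particles, with wrappers to activate (with optional reset), deactivate and set a named boolean parameter.
; Each wrapper passes the result of GetNiagaraParticles to a ManagedParticles call marked defunct.
; The getNiagaraParticles helper builds a one-element gameplay tag container from a Niagara tag variable, so particles are targeted by tag.
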(fn ActivateManagedNiagara
  (params (Reset :type Boolean))
  (ManagedParticles|Systems|%28defunct%29ActivateManagedParticles 0 (ManagedParticles|GetNiagaraParticles 0) Reset))

(fn DeactivateManagedNiagara
  (ManagedParticles|Systems|%28defunct%29DeactivateManagedParticles 0 (ManagedParticles|GetNiagaraParticles 0)))

(fn getNiagaraParticles
  (return (GameplayTags|MakeGameplayTagContainerfromArray (Utilities|Array|MakeArray (LocalVariables|GetNiagaraTag)))))

(fn SetManagedParticlesParameter(Bool)
  (params (Name :type Name) (Value :type Boolean))
  (ManagedParticles|Parameters|%28defunct%29SetBooleanParameter%28ManagedParticles%29 0 (ManagedParticles|GetNiagaraParticles 0) Name Value))
